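# Niagara skydiving contrail "DieselSmoke" that places two engine meshes on sampled sockets, each with a light and sprite glow that fades in.
# Smoke and spark emitters partition particles across both sockets, spawning at offsets rotated by mesh orientation.
# Smoke cools from hot orange to dark, with front-end color and time-of-day alpha adjustments.
# Sparks respond to glide lean, with GPU High and cheaper CPU MedLow variants, the latter without curl noise.
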NiagaraSystem /BRCosmetics/Athena/Effects/Particles/Skydiving/NS_Contrail_DieselSmoke/NS_Contrail_DieselSmoke
dependencies {
 usedDataInterfaces [
  &NiagaraDataInterfaceCurve
  &NiagaraDataInterfaceSpriteRendererInfo
  &NiagaraDataInterfaceColorCurve
  &NiagaraDataInterfaceSocketReader
 ]
 usedDynamicInputs []
 usedModules [
  &self:SystemSpawnScript.SystemScriptSource.SystemScriptGraph.NiagaraNodeAssignment_0.SetVariables_#1
  &SystemState
 ]
}
emitters {
 Engines {
  emitterUpdateScript [
   &EmitterState(Life Cycle Mode=::NewEnumerator1, Loop Behavior=::NewEnumerator1, Loop Duration Mode=::NewEnumerator1, UseLoopDelay=false)
   &SpawnBurst_Instantaneous(Spawn Count=2, Use Spawn Probability=false, UseLoopCountLimit=false)
  ]
  particleSpawnScript [
   &PartitionParticles(ID to Partition:NiagaraInt32=&ReturnExecIndex())
   &InitializeParticle(
    Color Mode=::NewEnumerator0
    Mesh Scale Mode=::NewEnumerator3
    Mesh Uniform Scale=0.3
    Sprite Size Mode=::NewEnumerator3
    Uniform Sprite Size=42
    UsePositionOffset=false
    Write Material Random=true
   )
   &InitialMeshOrientation(
    Mesh Orientation Mode=::NewEnumerator2
    Orientation Coordinate Space=::Local
    Orientation Vector=<1 1 0>
    Rotation=<0 0 -0.15>
    Use Rotation Vector=true
   )
   &CameraOffset(Camera Offset Amount=1)
   /SetVariables(Particles.CameraOffsetGlow=60) {isSetParameters=true}
  ]
  particleUpdateScript [
   &SocketLocation(
    Interpolate Socket Sampling=true
    Orientation Sampling=::NewEnumerator1
    SamplePosOffset=true
    Sampled Orientation Coordinate Space:ENiagaraCoordinateSpace=::Local
    Sampled Position Offset=&MergeHelper_VectorSwitch(0=<10 0 0>, 1=<-10 0 0>, 2=?, 3=?, Index=$Particles.PartitionParticles.Partition, Num=::NewEnumerator0)
    Socket Index=$Particles.PartitionParticles.Partition
    Socket Reader=$System.Socket Reader
    Socket Sampling Mode=::NewEnumerator3
    VelocitySpeedLimit=false
   )
   &ParticleState(Kill Particles When Lifetime Has Elapsed=false, Let Infinitely Lived Particles Die When Emitter Deactivates=true)
   /SetVariables(Particles.LightPosition:NiagaraPosition=&AddVectorToPosition(
     Position=$Particles.Position
     Vector=&RotateVectorByQuaternion(Vector To Rotate=<0 -18 36>, Quaternion=$Particles.MeshOrientation)
    )) {isSetParameters=true}
   /SetVariables(Particles.GlowColor=&MultiplyLinearColor(
     Linear Color=<rgba 2 0.6333 0 1>
     Scale Factor=&MakeVector4FromVectorAndFloat(
      "Vector (RGB)"=<1 1 1>
      "Float (Alpha)"=&FloatFromCurve(FloatCurve=DataInterface() {propertyValues {Curve {keys [<at 0 0>,<at 0.2 1>]}}}, CurveIndex=$Engine.System.Age, Scale Curve=0.15)
     )
    )) {isSetParameters=true}
   /SetVariables(Particles.LightColor=&MultiplyLinearColor(
     Linear Color=$Particles.GlowColor
     Scale Factor=&MakeVector4FromVectorAndFloat(
      "Vector (RGB)"=<1 0.9 0.5>
      "Float (Alpha)"=&FN_FrontEndFloatModifier(Gameplay=200, Front End=130)
     )
    )) {isSetParameters=true}
  ]
  rendererClasses [
   &NiagaraMeshRendererProperties
   &NiagaraLightRendererProperties
   &NiagaraSpriteRendererProperties
  ]
  renderers [
   {rendererClass=&NiagaraMeshRendererProperties, rendererIndex=0}
   {rendererClass=&NiagaraLightRendererProperties, rendererIndex=1}
   {rendererClass=&NiagaraSpriteRendererProperties, rendererIndex=2}
  ]
 }
 Smoke {
  emitterUpdateScript [
   &EmitterState(Life Cycle Mode=::NewEnumerator1, Loop Behavior=::NewEnumerator1, Loop Duration Mode=::NewEnumerator1, UseLoopDelay=false)
   &SpawnRate(SpawnRate=80, Use Spawn Probability=false)
  ]
  particleSpawnScript [
   &PartitionParticles()
   /SetVariables(Particles.VisibilityTag=&RandomRangeInt(Minimum=0, Maximum=1)) {isSetParameters=true}
   &InitializeParticle(
    Color=<rgba 1 0.96 0.92 1>
    Lifetime=&RandomRangeFloat(Minimum=0.4, Maximum=0.8)
    Mesh Scale Mode=::NewEnumerator3
    Mesh Uniform Scale=0.3
    Sprite Rotation Mode=::NewEnumerator2
    Sprite Size Mode=::NewEnumerator3
    Uniform Sprite Size=&RandomRangeFloat(Minimum=4, Maximum=7)
    UsePositionOffset=false
    Write Material Random=true
   )
   &SocketLocation(
    Orientation Sampling=::NewEnumerator1
    SamplePosOffset=true
    Sampled Orientation Coordinate Space:ENiagaraCoordinateSpace=::Local
    Sampled Position Offset=&MergeHelper_VectorSwitch(0=<10 0 0>, 1=<-10 0 0>, 2=?, 3=?, Index=$Particles.PartitionParticles.Partition, Num=::NewEnumerator0)
    Socket Index=$Particles.PartitionParticles.Partition
    Socket Reader=$System.Socket Reader
    Socket Sampling Mode=::NewEnumerator3
    VelocitySpeedLimit=false
   )
   &InitialMeshOrientation(
    Mesh Orientation Mode=::NewEnumerator2
    Orientation Coordinate Space=::Local
    Orientation Vector=<1 1 0>
    Rotation=<0 0 -0.15>
    Use Rotation Vector=true
   )
   &ShapeLocation(MaskSpawnGroup=false, Override Seed=false, Sphere Radius=2, UseOwnerScale=true)
   /SetVariables(Particles.Position=&AddVectorToPosition(
     Position=$Particles.Position
     Vector=&RotateVectorByQuaternion(Vector To Rotate=<0 -12 24>, Quaternion=$Particles.MeshOrientation)
    )) {isSetParameters=true, module=SetVariables_#2}
   &AddVelocity(Velocity=<0 -150 350>, Velocity Speed Scale=&RandomRangeFloat(Minimum=0.8, Maximum=1))
   &CameraOffset(Camera Offset Amount=7)
  ]
  particleUpdateScript [
   &ParticleState()
   &ScaleSpriteSize(Uniform Curve Sprite Scale=DataInterface() {propertyValues {Curve {keys [
        {arriveTangent=-2.2836, interpMode=RCIM_Cubic, leaveTangent=-2.2836, tangentMode=RCTM_User, time=0, value=3}
        {arriveTangent=13.3333, interpMode=RCIM_Cubic, leaveTangent=13.3333, time=0.65, value=9.8}
        {arriveTangent=1.706, interpMode=RCIM_Cubic, leaveTangent=1.706, tangentMode=RCTM_User, time=0.9, value=15}
       ]}}})
   &Drag(Use Linear Drag=true, Use Rotational Drag=true)
   &SolveForcesAndVelocity(Clamp Velocity=false, Limit Acceleration=false)
   &DynamicMaterialParameters(
    Index 0 Param 1=&FloatFromCurve(FloatCurve=DataInterface() {propertyValues {Curve {keys [<at 0.5 1 RCIM_Cubic>,<at 0.9993 -0.0385 RCIM_Cubic>]}}})
    Index 0 Param 2=1
    Index 0 Param 3=1
    Index 0 Param 4=0.5
    Module.Param0WriteEnabled=true
    Module.Param1WriteEnabled=true
    Module.Param2WriteEnabled=true
    Module.Param3WriteEnabled=true
   )
   &SubUVAnimation(
    Frames Per Second=30
    Playback Mode:ENiagara_FPSPlayrate=::NewEnumerator1
    Playback Scale Curve=DataInterface() {propertyValues {Curve {keys [
        {interpMode=RCIM_Cubic, tangentMode=RCTM_User, time=0.25, value=1}
        {arriveTangent=-1.2656, interpMode=RCIM_Cubic, leaveTangent=-1.2656, tangentMode=RCTM_User, time=1, value=0}
       ]}}}
    "Sprite Renderer "=DataInterface() {propertyValues {SpriteRenderer=&self:Smoke_1.NiagaraSpriteRendererProperties_2}}
    SubUV Animation Mode:ENiagara_SubUVLookupModeV2=::NewEnumerator3
    UseEndFrame=false
    UsePlayRateScale=false
    UseStartFrame=false
   )
   &ScaleColor(
    Scale Alpha=&FloatFromCurve(
     FloatCurve=DataInterface() {propertyValues {Curve {keys [
         {arriveTangent=2.6574, interpMode=RCIM_Cubic, leaveTangent=2.6574, tangentMode=RCTM_User, time=0, value=0}
         {interpMode=RCIM_Cubic, tangentMode=RCTM_User, time=0.15, value=1}
         {interpMode=RCIM_Cubic, tangentMode=RCTM_User, time=1, value=0}
        ]}}}
     Scale Curve=&FNTimeOfDayScalar(Night Scale=1, Morning Scale=0.9, Day Scale=0.75, Evening Scale=0.9)
    )
    Scale RGB=&FN_FrontEndVector3Modifier(
     Gameplay=&MakeVectorFromLinearColorRGB(Color=&ColorFromCurve(ColorCurve=DataInterface() {propertyValues {
         AlphaCurve {keys [<at 0 1 RCIM_Cubic>]}
         BlueCurve {keys [<at 0.0776 0.91>, <at 0.2226 0.18>, {interpMode=RCIM_Cubic, leaveTangent=0.387, time=0.382, value=0.16}]}
         GreenCurve {keys [<at 0.0776 3.9325>, <at 0.2226 0.65>, {interpMode=RCIM_Cubic, leaveTangent=-0.0209, time=0.382, value=0.16}]}
         RedCurve {keys [<at 0.0776 13>, <at 0.2226 3>, {interpMode=RCIM_Cubic, leaveTangent=-0.8368, time=0.382, value=0.16}]}
        }}))
     Front End=&MakeVectorFromLinearColorRGB(Color=&ColorFromCurve(ColorCurve=DataInterface() {propertyValues {
         AlphaCurve {keys [<at 0 1 RCIM_Cubic>]}
         BlueCurve {keys [<at 0.0776 0.65>, <at 0.2417 0.11>, {interpMode=RCIM_Cubic, leaveTangent=0.387, time=0.384, value=0.07}]}
         GreenCurve {keys [<at 0.0776 3.7375>, <at 0.2417 0.4583>, {interpMode=RCIM_Cubic, leaveTangent=-0.0209, time=0.384, value=0.07}]}
         RedCurve {keys [<at 0.0776 13>, <at 0.2417 2.2>, {interpMode=RCIM_Cubic, leaveTangent=-0.8368, time=0.384, value=0.07}]}
        }}))
    )
    ScaleA=true
    ScaleRGB=true
   )
  ]
  rendererClasses [&NiagaraSpriteRendererProperties]
  renderers [
   {rendererClass=&NiagaraSpriteRendererProperties}
   {rendererClass=&NiagaraSpriteRendererProperties, rendererIndex=1}
  ]
 }
 SparksHigh {
  simTarget=GPUComputeSim
  emitterUpdateScript [
   &EmitterState(Life Cycle Mode=::NewEnumerator1, Loop Behavior=::NewEnumerator1, Loop Duration Mode=::NewEnumerator1, UseLoopDelay=false)
   &SpawnRate(
    SpawnRate=&Multiply_Float(A=50, B=&Lerp_Float(A=0.8, B=1.2, Alpha=$User.ParaGlideLeanAlpha))
    Use Spawn Probability=false
   )
  ]
  particleSpawnScript [
   &PartitionParticles()
   &InitializeParticle(
    Color=<rgba 1 0.6111 0.3 1>
    Lifetime=&RandomRangeFloat(Minimum=0.25, Maximum=0.8)
    Sprite Rotation Mode=::NewEnumerator2
    Sprite Size=&Multiply_Vector2D_ByFloat(Float=&RandomRangeFloat(Minimum=0.8, Maximum=1.5), Vector2D=<0.5 3>)
    Sprite Size Mode=::NewEnumerator4
    UsePositionOffset=false
    Write Material Random=true
   )
   &SocketLocation(
    Orientation Sampling=::NewEnumerator1
    SamplePosOffset=true
    Sampled Orientation Coordinate Space:ENiagaraCoordinateSpace=::Local
    Sampled Position Offset=&MergeHelper_VectorSwitch(0=<10 0 0>, 1=<-10 0 0>, 2=?, 3=?, Index=$Particles.PartitionParticles.Partition, Num=::NewEnumerator0)
    Socket Index=$Particles.PartitionParticles.Partition
    Socket Reader=$System.Socket Reader
    Socket Sampling Mode=::NewEnumerator3
    VelocitySpeedLimit=false
   )
   &InitialMeshOrientation(
    Mesh Orientation Mode=::NewEnumerator2
    Orientation Coordinate Space=::Local
    Orientation Vector=<1 1 0>
    Rotation=<0 0 -0.15>
    Use Rotation Vector=true
   )
   /SetVariables(Particles.Position=&AddVectorToPosition(
     Position=$Particles.Position
     Vector=&RotateVectorByQuaternion(Vector To Rotate=<0 -20 40>, Quaternion=$Particles.MeshOrientation)
    )) {isSetParameters=true, module=SetVariables_#2}
   &ShapeLocation(MaskSpawnGroup=false, Override Seed=false, Sphere Radius=11, Sphere Surface Distribution=0.2381, UseOwnerScale=true)
   &AddVelocity(
    Module.Use Falloff Distance=false
    Origin Offset=&RotateVectorByQuaternion(Vector To Rotate=<0 -5 10>, Quaternion=$Particles.MeshOrientation)
    Override Seed=false
    Velocity Mode=::NewEnumerator1
    Velocity Origin=$Output.SocketLocation.SampledPosition
    Velocity Speed=&RandomRangeFloat(Minimum=100, Maximum=250)
   )
   &InheritVelocity(
    Inherited Velocity Amount Scale=&VectorFromFloat(Value=0.5)
    Inherited Velocity Speed Limit=-150
    Speed Limit=true
    UseSpeedThreshold=false
    Velocity Source=$Output.SocketLocation.SampledVelocity
   )
  ]
  particleUpdateScript [
   &CameraOffset(Camera Offset Amount=&Lerp_Float(
     A=-3
     B=10
     Alpha=&FloatFromCurve(FloatCurve=DataInterface() {propertyValues {Curve {keys [<at 0.2 0>,<at 0.7 1>]}}})
    ))
   &ParticleState()
   &ScaleSpriteSize(Uniform Curve Sprite Scale=DataInterface() {propertyValues {Curve {keys [<at 0 0>, {tangentMode=RCTM_User, time=0.1, value=1}, <at 0.85 1>, {tangentMode=RCTM_User, time=1, value=0}]}}})
   &Drag(
    Drag=&Multiply_Float(
     A=&RandomRangeFloat(Minimum=2.5, Maximum=4)
     B=&Lerp_Float(A=1, B=0.5, Alpha=$User.ParaGlideLeanAlpha)
    )
    Use Linear Drag=true
    Use Rotational Drag=true
   )
   &CurlNoiseForce(
    Bias Noise=false
    Noise Frequency=270
    Noise Strength=&Multiply_Float(
     A=&Lerp_Float(A=1500, B=800, Alpha=$User.ParaGlideLeanAlpha)
     B=&FloatFromCurve(FloatCurve=DataInterface() {propertyValues {Curve {keys [
          {arriveTangent=2.1881, interpMode=RCIM_Cubic, leaveTangent=2.1881, tangentMode=RCTM_User, time=0, value=0}
          {interpMode=RCIM_Cubic, tangentMode=RCTM_User, time=0.5, value=1}
         ]}}})
    )
    Pan Noise=true
    Pan Noise Field=<0.1 0.2 0.1>
   )
   &AccelerationForce(Acceleration=<0 0 250>)
   &SolveForcesAndVelocity(Clamp Velocity=false, Limit Acceleration=false)
   &ScaleSpriteSizeBySpeed(Max Scale Factor=<0.5 2>, Min Scale Factor=<1 0.5>, Velocity Threshold=800)
   &ScaleColor(
    Scale Alpha=&FloatFromCurve(FloatCurve=DataInterface() {propertyValues {Curve {keys [
         {arriveTangent=2.6751, interpMode=RCIM_Cubic, leaveTangent=2.6751, tangentMode=RCTM_User, time=0, value=0}
         {interpMode=RCIM_Cubic, tangentMode=RCTM_User, time=0.2, value=1}
         {interpMode=RCIM_Cubic, tangentMode=RCTM_User, time=1, value=0}
        ]}}})
    Scale RGB=&MakeVectorFromLinearColorRGB(Color=&ColorFromCurve(ColorCurve=DataInterface() {propertyValues {
        AlphaCurve {keys [<at 0 1 RCIM_Cubic>]}
        BlueCurve {keys [<at 0.1262 1.98>,<at 0.2984 0.8816>,<at 0.8777 0.02>]}
        GreenCurve {keys [<at 0.1262 9.988>,<at 0.2984 4.4773>,<at 0.8777 0.216>]}
        RedCurve {keys [<at 0.1262 22>,<at 0.2984 10>,<at 0.8777 1>]}
       }}))
    ScaleA=false
    ScaleRGB=true
   )
   &DynamicMaterialParameters(
    Index 0 Param 1=1
    Index 0 Param 2=3
    Index 0 Param 3=1
    Module.Index0Float0Write=true
    Module.Index0Float1Write=true
    Module.Index0Float2Write=true
    Module.Index0Float3Write=false
   )
  ]
  rendererClasses [&NiagaraSpriteRendererProperties]
  renderers [{rendererClass=&NiagaraSpriteRendererProperties}]
 }
 SparksMedLow {
  emitterUpdateScript [
   &EmitterState(Life Cycle Mode=::NewEnumerator1, Loop Behavior=::NewEnumerator1, Loop Duration Mode=::NewEnumerator1, UseLoopDelay=false)
   &SpawnRate(
    SpawnRate=&Multiply_Float(A=25, B=&Lerp_Float(A=0.8, B=1.2, Alpha=$User.ParaGlideLeanAlpha))
    Use Spawn Probability=false
   )
  ]
  particleSpawnScript [
   &PartitionParticles()
   &InitializeParticle(
    Color=<rgba 1 0.6111 0.3 1>
    Lifetime=&RandomRangeFloat(Minimum=0.25, Maximum=0.8)
    Sprite Rotation Mode=::NewEnumerator2
    Sprite Size=&Multiply_Vector2D_ByFloat(Float=&RandomRangeFloat(Minimum=0.8, Maximum=1.5), Vector2D=<0.5 3>)
    Sprite Size Mode=::NewEnumerator4
    UsePositionOffset=false
    Write Material Random=true
   )
   &SocketLocation(
    Orientation Sampling=::NewEnumerator1
    SamplePosOffset=true
    Sampled Orientation Coordinate Space:ENiagaraCoordinateSpace=::Local
    Sampled Position Offset=&MergeHelper_VectorSwitch(0=<10 0 0>, 1=<-10 0 0>, 2=?, 3=?, Index=$Particles.PartitionParticles.Partition, Num=::NewEnumerator0)
    Socket Index=$Particles.PartitionParticles.Partition
    Socket Reader=$System.Socket Reader
    Socket Sampling Mode=::NewEnumerator3
    VelocitySpeedLimit=false
   )
   &InitialMeshOrientation(
    Mesh Orientation Mode=::NewEnumerator2
    Orientation Coordinate Space=::Local
    Orientation Vector=<1 1 0>
    Rotation=<0 0 -0.15>
    Use Rotation Vector=true
   )
   /SetVariables(Particles.Position=&AddVectorToPosition(
     Position=$Particles.Position
     Vector=&RotateVectorByQuaternion(Vector To Rotate=<0 -20 40>, Quaternion=$Particles.MeshOrientation)
    )) {isSetParameters=true, module=SetVariables_#2}
   &ShapeLocation(MaskSpawnGroup=false, Override Seed=false, Sphere Radius=11, Sphere Surface Distribution=0.2381, UseOwnerScale=true)
   &AddVelocity(
    Module.Use Falloff Distance=false
    Origin Offset=&RotateVectorByQuaternion(Vector To Rotate=<0 -5 10>, Quaternion=$Particles.MeshOrientation)
    Override Seed=false
    Velocity Mode=::NewEnumerator1
    Velocity Origin=$Output.SocketLocation.SampledPosition
    Velocity Speed=&RandomRangeFloat(Minimum=100, Maximum=250)
   )
   &InheritVelocity(
    Inherited Velocity Amount Scale=&VectorFromFloat(Value=0.5)
    Inherited Velocity Speed Limit=-150
    Speed Limit=true
    UseSpeedThreshold=false
    Velocity Source=$Output.SocketLocation.SampledVelocity
   )
  ]
  particleUpdateScript [
   &CameraOffset(Camera Offset Amount=&Lerp_Float(
     A=-3
     B=10
     Alpha=&FloatFromCurve(FloatCurve=DataInterface() {propertyValues {Curve {keys [<at 0.2 0>,<at 0.7 1>]}}})
    ))
   &ParticleState()
   &ScaleSpriteSize(Uniform Curve Sprite Scale=DataInterface() {propertyValues {Curve {keys [<at 0 0>, {tangentMode=RCTM_User, time=0.1, value=1}, <at 0.85 1>, {tangentMode=RCTM_User, time=1, value=0}]}}})
   &Drag(
    Drag=&Multiply_Float(
     A=&RandomRangeFloat(Minimum=2.5, Maximum=4)
     B=&Lerp_Float(A=1, B=0.5, Alpha=$User.ParaGlideLeanAlpha)
    )
    Use Linear Drag=true
    Use Rotational Drag=true
   )
   &CurlNoiseForce(
    Bias Noise=false
    Noise Frequency=350
    Noise Strength=&Multiply_Float(
     A=&Lerp_Float(A=1500, B=800, Alpha=$User.ParaGlideLeanAlpha)
     B=&FloatFromCurve(FloatCurve=DataInterface() {propertyValues {Curve {keys [
          {arriveTangent=2.1881, interpMode=RCIM_Cubic, leaveTangent=2.1881, tangentMode=RCTM_User, time=0, value=0}
          {interpMode=RCIM_Cubic, tangentMode=RCTM_User, time=0.5, value=1}
         ]}}})
    )
    Pan Noise=true
    Pan Noise Field=<0.1 0.2 0.1>
   ) {enabled=false}
   &AccelerationForce(Acceleration=<0 0 250>)
   &SolveForcesAndVelocity(Clamp Velocity=false, Limit Acceleration=false)
   &ScaleSpriteSizeBySpeed(Max Scale Factor=<0.5 2>, Min Scale Factor=<1 0.5>, Velocity Threshold=800)
   &ScaleColor(
    Scale Alpha=&FloatFromCurve(FloatCurve=DataInterface() {propertyValues {Curve {keys [
         {arriveTangent=2.6751, interpMode=RCIM_Cubic, leaveTangent=2.6751, tangentMode=RCTM_User, time=0, value=0}
         {interpMode=RCIM_Cubic, tangentMode=RCTM_User, time=0.2, value=1}
         {interpMode=RCIM_Cubic, tangentMode=RCTM_User, time=1, value=0}
        ]}}})
    Scale RGB=&MakeVectorFromLinearColorRGB(Color=&ColorFromCurve(ColorCurve=DataInterface() {propertyValues {
        AlphaCurve {keys [<at 0 1 RCIM_Cubic>]}
        BlueCurve {keys [<at 0.1262 1.98>,<at 0.2984 0.8816>,<at 0.8777 0.02>]}
        GreenCurve {keys [<at 0.1262 9.988>,<at 0.2984 4.4773>,<at 0.8777 0.216>]}
        RedCurve {keys [<at 0.1262 22>,<at 0.2984 10>,<at 0.8777 1>]}
       }}))
    ScaleA=false
    ScaleRGB=true
   )
   &DynamicMaterialParameters(
    Index 0 Param 1=1
    Index 0 Param 2=3
    Index 0 Param 3=1
    Module.Index0Float0Write=true
    Module.Index0Float1Write=true
    Module.Index0Float2Write=true
    Module.Index0Float3Write=false
   )
  ]
  rendererClasses [&NiagaraSpriteRendererProperties]
  renderers [{rendererClass=&NiagaraSpriteRendererProperties}]
 }
}
summary {
 systemName=NS_Contrail_DieselSmoke
 emitters [
  {
   emitterName=Engines
   rendererClasses [
    &NiagaraMeshRendererProperties
    &NiagaraLightRendererProperties
    &NiagaraSpriteRendererProperties
   ]
  }
  {emitterName=Smoke, rendererClasses [&NiagaraSpriteRendererProperties]}
  {emitterName=SparksHigh, simTarget=GPUComputeSim, rendererClasses [&NiagaraSpriteRendererProperties]}
  {emitterName=SparksMedLow, rendererClasses [&NiagaraSpriteRendererProperties]}
 ]
 userVariables [
  User.bIsFrontEnd {type=NiagaraBool}
  User.bIsFrontEndPreview {type=NiagaraBool}
  User.ParaGlideLeanAlpha {type=NiagaraFloat, defaultValue=0, description=""}
 ]
}
